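# Physics asset for the player's medium male WaterMolecules slide coat, defining capsule bodies for jacket-flap dynamics.
# Left and right flap chains (A, B, Ba, C) hang from MSTR anchor bones, joined by constraints with narrow swing angles and zero twist; the top of A-left is locked.
# Oversized leg and foot capsules, plus pelvis and spine capsules, form the body shapes.
# A huge root capsule sits below the origin, apparently as a ground proxy.
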PhysicsAsset /Game/Characters/Player/Male/Medium/Bodies/M_MED_WaterMolecules/Meshes/M_MED_WaterMolecules_Coat_Slide_Physics
bodies {
 MSTR_jacketA_l [{shapeType=Capsule, length=5, radius=3, shapeName=MSTR_jacketA_l_capsule, rotation=<rot -90 0 0>}]
 MSTR_jacketA_r [{shapeType=Capsule, length=5, radius=3, shapeName=MSTR_jacketA_r_capsule, rotation=<rot -90 0 0>}]
 MSTR_jacketB_l [{shapeType=Capsule, length=5, radius=3, shapeName=MSTR_jacketB_l_capsule, rotation=<rot -90 0 0>}]
 MSTR_jacketB_r [{shapeType=Capsule, length=5, radius=3, shapeName=MSTR_jacketB_r_capsule, rotation=<rot -90 0 0>}]
 MSTR_jacketBa_l [{shapeType=Capsule, length=5, radius=3, shapeName=MSTR_jacketBa_l_capsule, rotation=<rot -90 0 0>}]
 MSTR_jacketC_l [{shapeType=Capsule, length=5, radius=3, shapeName=MSTR_jacketC_l_capsule, rotation=<rot -90 0 0>}]
 MSTR_jacketC_r [{shapeType=Capsule, length=5, radius=3, shapeName=MSTR_jacketC_r_capsule, rotation=<rot -90 0 0>}]
 calf_l [{shapeType=Capsule, length=13.5, radius=18.5, shapeName=calf_l_capsule, center=<-20 -4.25 5.4>, rotation=<rot -90 0 0>}]
 calf_r [{shapeType=Capsule, length=12.5, radius=19.5, shapeName=calf_r_capsule, center=<20 2.6084 -6.9131>, rotation=<rot -90 180 180>}]
 dyn_jacketA_l_01 [{shapeType=Capsule, length=12.5, radius=5, shapeName=dyn_jacketA_l_01_capsule, center=<10 -5.6 3.8>, rotation=<rot 0 0 -104>}]
 dyn_jacketA_l_01A [{
   shapeType=Capsule
   length=12.5
   radius=5
   shapeName=dyn_jacketA_l_01A_capsule
   center=<10 -5.6896 5.5344>
   rotation=<rot 0.0002 0.0001 -103.686>
  }]
 dyn_jacketA_l_02 [{shapeType=Capsule, length=12.5, radius=5, shapeName=dyn_jacketA_l_02_capsule, center=<10 -5.6 3.3>, rotation=<rot 0.0002 0.0001 -103.686>}]
 dyn_jacketA_l_03 [{shapeType=Capsule, length=12.5, radius=5, shapeName=dyn_jacketA_l_03_capsule, center=<10 -5.6 3.8>, rotation=<rot 0 0 -104>}]
 dyn_jacketA_l_04 [{shapeType=Capsule, length=10, radius=7, shapeName=dyn_jacketA_l_04_capsule, center=<7 0 5.8>, rotation=<rot 0 0 -75>}]
 dyn_jacketA_l_05 [{
   shapeType=Capsule
   length=22.4901
   radius=8
   shapeName=dyn_jacketA_l_05_capsule
   center=<7.7 -6.6 6.8>
   rotation=<rot 0.0002 0.0002 -77.1214>
  }]
 dyn_jacketA_l_06 [{shapeType=Capsule, length=12.5, radius=5, shapeName=dyn_jacketA_l_06_capsule, center=<10 -5.6 2.5>, rotation=<rot 0 0 -104>}]
 dyn_jacketA_r_01 [{shapeType=Capsule, length=10, radius=5, shapeName=dyn_jacketA_r_01_capsule, center=<10 0 4.1>, rotation=<rot -90 0 0>}]
 dyn_jacketA_r_01A [{shapeType=Capsule, length=10, radius=5, shapeName=dyn_jacketA_r_01A_capsule, center=<10 0 6>, rotation=<rot -90 0 0>}]
 dyn_jacketA_r_02 [{shapeType=Capsule, length=10, radius=5, shapeName=dyn_jacketA_r_02_capsule, center=<10 0 3.6>, rotation=<rot -90 0 0>}]
 dyn_jacketA_r_03 [{shapeType=Capsule, length=10, radius=5, shapeName=dyn_jacketA_r_03_capsule, center=<10 0 4.8>, rotation=<rot -90 0 0>}]
 dyn_jacketA_r_04 [{shapeType=Capsule, length=10, radius=5, shapeName=dyn_jacketA_r_04_capsule, center=<10 0 4>, rotation=<rot -90 0 0>}]
 dyn_jacketA_r_05 [{shapeType=Capsule, length=10, radius=5, shapeName=dyn_jacketA_r_05_capsule, center=<10 0 2.7>, rotation=<rot -90 0 0>}]
 dyn_jacketA_r_06 [{shapeType=Capsule, length=10, radius=5, shapeName=dyn_jacketA_r_06_capsule, center=<10 0 4.5>, rotation=<rot -90 0 0>}]
 dyn_jacketB_l_01 [{shapeType=Capsule, length=30.809, radius=5, shapeName=dyn_jacketB_l_01_capsule, center=<10 0 4.4>, rotation=<rot 0 0.0001 110.6919>}]
 dyn_jacketB_l_02 [{shapeType=Capsule, length=30.809, radius=5, shapeName=dyn_jacketB_l_02_capsule, center=<10 0 4.3>, rotation=<rot 0 0.0001 110.6919>}]
 dyn_jacketB_l_03 [{shapeType=Capsule, length=30.809, radius=5, shapeName=dyn_jacketB_l_03_capsule, center=<10 0 4>, rotation=<rot 0 0.0001 110.6919>}]
 dyn_jacketB_l_04 [{shapeType=Capsule, length=30.809, radius=5, shapeName=dyn_jacketB_l_04_capsule, center=<10 0 4>, rotation=<rot 0 0.0001 110.6919>}]
 dyn_jacketB_l_05 [{shapeType=Capsule, length=31, radius=5, shapeName=dyn_jacketB_l_05_capsule, center=<10 0.6 3.6>, rotation=<rot 0 0 124>}]
 dyn_jacketB_l_06 [{shapeType=Capsule, length=30.809, radius=5, shapeName=dyn_jacketB_l_06_capsule, center=<10 0 2>, rotation=<rot 0 0.0001 110.6919>}]
 dyn_jacketB_r_01 [{shapeType=Capsule, length=30.809, radius=5, shapeName=dyn_jacketB_r_01_capsule, center=<10 0 6>, rotation=<rot 0 0 90>}]
 dyn_jacketB_r_02 [{shapeType=Capsule, length=30.809, radius=5, shapeName=dyn_jacketB_r_02_capsule, center=<10 0 6>, rotation=<rot 0 0 90>}]
 dyn_jacketB_r_03 [{shapeType=Capsule, length=30.809, radius=5, shapeName=dyn_jacketB_r_03_capsule, center=<10 0 6>, rotation=<rot 0 0 90>}]
 dyn_jacketB_r_04 [{shapeType=Capsule, length=30.809, radius=5, shapeName=dyn_jacketB_r_04_capsule, center=<10 0 6>, rotation=<rot 0 0 90>}]
 dyn_jacketB_r_05 [{shapeType=Capsule, length=30.809, radius=5, shapeName=dyn_jacketB_r_05_capsule, center=<10 0 6>, rotation=<rot 0 0 90>}]
 dyn_jacketB_r_06 [{shapeType=Capsule, length=30.809, radius=5, shapeName=dyn_jacketB_r_06_capsule, center=<10 0 6>, rotation=<rot 0 0 90>}]
 dyn_jacketBa_l_01 [{shapeType=Capsule, length=37.6252, radius=5, shapeName=dyn_jacketBa_l_01_capsule, center=<10 0 6>, rotation=<rot 0 0 90>}]
 dyn_jacketBa_l_02 [{shapeType=Capsule, length=37.6252, radius=5, shapeName=dyn_jacketBa_l_02_capsule, center=<10 0 6>, rotation=<rot 0 0 90>}]
 dyn_jacketBa_l_03 [{shapeType=Capsule, length=37.6252, radius=5, shapeName=dyn_jacketBa_l_03_capsule, center=<10 0 6>, rotation=<rot 0 0 90>}]
 dyn_jacketBa_l_04 [{shapeType=Capsule, length=37.6252, radius=5, shapeName=dyn_jacketBa_l_04_capsule, center=<10 0 3>, rotation=<rot 0 0 90>}]
 dyn_jacketBa_l_05 [{shapeType=Capsule, length=37.6252, radius=5, shapeName=dyn_jacketBa_l_05_capsule, center=<10 0 1.3>, rotation=<rot 0 0 90>}]
 dyn_jacketBa_l_06 [{shapeType=Capsule, length=35.1252, radius=7.5, shapeName=dyn_jacketBa_l_06_capsule, center=<10 0 6>, rotation=<rot 0 0 90>}]
 dyn_jacketC_l_01 [{shapeType=Capsule, length=30.809, radius=5, shapeName=dyn_jacketC_l_01_capsule, center=<10 0 6>, rotation=<rot 0 0 90>}]
 dyn_jacketC_l_02 [{shapeType=Capsule, length=30.809, radius=5, shapeName=dyn_jacketC_l_02_capsule, center=<10 0 6>, rotation=<rot 0 0 90>}]
 dyn_jacketC_l_03 [{shapeType=Capsule, length=30.809, radius=5, shapeName=dyn_jacketC_l_03_capsule, center=<10 0 6>, rotation=<rot 0 0 90>}]
 dyn_jacketC_l_04 [{shapeType=Capsule, length=27.309, radius=8.5, shapeName=dyn_jacketC_l_04_capsule, center=<10 0 3.1488>, rotation=<rot 0 0 90>}]
 dyn_jacketC_l_05 [{shapeType=Capsule, length=26.809, radius=9, shapeName=dyn_jacketC_l_05_capsule, center=<10 0 8.244>, rotation=<rot 0 0 90>}]
 dyn_jacketC_l_06 [{shapeType=Capsule, length=27.309, radius=8.5, shapeName=dyn_jacketC_l_06_capsule, center=<10 0 7.7798>, rotation=<rot 0 0 90>}]
 dyn_jacketC_r_01 [{shapeType=Capsule, length=30.809, radius=5, shapeName=dyn_jacketC_r_01_capsule, center=<10 0 6>, rotation=<rot 0 0 90>}]
 dyn_jacketC_r_02 [{shapeType=Capsule, length=30.809, radius=5, shapeName=dyn_jacketC_r_02_capsule, center=<10 0 6>, rotation=<rot 0 0 90>}]
 dyn_jacketC_r_03 [{shapeType=Capsule, length=30.809, radius=5, shapeName=dyn_jacketC_r_03_capsule, center=<10 0 6>, rotation=<rot 0 0 90>}]
 dyn_jacketC_r_04 [{shapeType=Capsule, length=30.809, radius=5, shapeName=dyn_jacketC_r_04_capsule, center=<10 0 5.1>, rotation=<rot 0 0 90>}]
 dyn_jacketC_r_05 [{shapeType=Capsule, length=30.809, radius=5, shapeName=dyn_jacketC_r_05_capsule, center=<10 0 3.5>, rotation=<rot 0 0 90>}]
 dyn_jacketC_r_06 [{shapeType=Capsule, length=30.809, radius=5, shapeName=dyn_jacketC_r_06_capsule, center=<10 0 6>, rotation=<rot 0 0 90>}]
 foot_l [{shapeType=Capsule, length=20, radius=17, shapeName=foot_l_capsule, center=<0 -3 0>, rotation=<rot -90 0 0>}]
 foot_r [{shapeType=Capsule, length=20, radius=17, shapeName=foot_r_capsule, center=<0 3 0>, rotation=<rot -90 0 0>}]
 pelvis [
  {shapeType=Capsule, length=7.4, radius=12.5, shapeName=pelvis_capsule, center=<0 -1.3 0>}
  {shapeType=Capsule, length=8.5, radius=14, shapeName=pelvis_capsule, center=<-9.817 -2.0887 0>}
 ]
 root [{shapeType=Capsule, length=300, radius=300, shapeName=root_capsule, center=<0 0 -295>, rotation=<rot -90 180 90>}]
 spine_02 [{shapeType=Capsule, length=5.8, radius=12.3, shapeName=spine_02_capsule, center=<0 0.4 0>}]
 spine_03 [{shapeType=Capsule, length=7.8, radius=11.8, shapeName=spine_03_capsule, center=<0 0.4 0>}]
 thigh_l [{shapeType=Capsule, length=26.5, radius=15.5, shapeName=thigh_l_capsule, center=<-24.9858 -2.7988 4.5365>, rotation=<rot -85 5 0>}]
 thigh_r [{shapeType=Capsule, length=25, radius=16.5, shapeName=thigh_r_capsule, center=<25 2.8 -4.7>, rotation=<rot -85 5 0>}]
}
constraints [
 {
  bone1Name=foot_r
  bone2Name=calf_r
  swing1LimitDegrees=45
  swing1Motion=Free
  swing2LimitDegrees=45
  swing2Motion=Free
  twistLimitDegrees=45
  twistMotion=Free
 }
 {
  bone1Name=MSTR_jacketA_l
  bone2Name=spine_03
  swing1LimitDegrees=45
  swing1Motion=Free
  swing2LimitDegrees=45
  swing2Motion=Free
  twistLimitDegrees=45
  twistMotion=Free
 }
 {
  bone1Name=dyn_jacketA_l_01
  bone2Name=dyn_jacketA_l_01A
  swing1LimitDegrees=20
  swing1Motion=Locked
  swing2LimitDegrees=20
  swing2Motion=Locked
  twistLimitDegrees=0
  twistMotion=Locked
 }
 {
  bone1Name=dyn_jacketA_l_02
  bone2Name=dyn_jacketA_l_01
  swing1LimitDegrees=20
  swing1Motion=Locked
  swing2LimitDegrees=20
  swing2Motion=Locked
  twistLimitDegrees=0
  twistMotion=Locked
 }
 {bone1Name=dyn_jacketA_l_03, bone2Name=dyn_jacketA_l_02, swing1LimitDegrees=20, swing2LimitDegrees=20, twistLimitDegrees=0}
 {bone1Name=dyn_jacketA_l_04, bone2Name=dyn_jacketA_l_03, swing1LimitDegrees=20, swing2LimitDegrees=20, twistLimitDegrees=0}
 {bone1Name=dyn_jacketA_r_01, bone2Name=dyn_jacketA_r_01A, swing1LimitDegrees=10, swing2LimitDegrees=10, twistLimitDegrees=0}
 {
  bone1Name=MSTR_jacketB_l
  bone2Name=spine_03
  swing1LimitDegrees=45
  swing1Motion=Free
  swing2LimitDegrees=45
  swing2Motion=Free
  twistLimitDegrees=45
  twistMotion=Free
 }
 {bone1Name=dyn_jacketB_l_01, bone2Name=MSTR_jacketB_l, swing1LimitDegrees=20, swing2LimitDegrees=20, twistLimitDegrees=0}
 {bone1Name=dyn_jacketB_l_02, bone2Name=dyn_jacketB_l_01, swing1LimitDegrees=20, swing2LimitDegrees=20, twistLimitDegrees=0}
 {bone1Name=dyn_jacketB_l_03, bone2Name=dyn_jacketB_l_02, swing1LimitDegrees=20, swing2LimitDegrees=20, twistLimitDegrees=0}
 {bone1Name=dyn_jacketB_l_04, bone2Name=dyn_jacketB_l_03, swing1LimitDegrees=20, swing2LimitDegrees=20, twistLimitDegrees=0}
 {bone1Name=dyn_jacketB_r_01, bone2Name=MSTR_jacketB_r, swing1LimitDegrees=20, swing2LimitDegrees=20, twistLimitDegrees=0}
 {bone1Name=dyn_jacketB_r_02, bone2Name=dyn_jacketB_r_01, swing1LimitDegrees=20, swing2LimitDegrees=20, twistLimitDegrees=0}
 {bone1Name=dyn_jacketB_r_03, bone2Name=dyn_jacketB_r_02, swing1LimitDegrees=20, swing2LimitDegrees=20, twistLimitDegrees=0}
 {bone1Name=dyn_jacketB_r_04, bone2Name=dyn_jacketB_r_03, swing1LimitDegrees=20, swing2LimitDegrees=20, twistLimitDegrees=0}
 {bone1Name=dyn_jacketBa_l_01, bone2Name=MSTR_jacketBa_l, swing1LimitDegrees=20, swing2LimitDegrees=20, twistLimitDegrees=0}
 {bone1Name=dyn_jacketBa_l_02, bone2Name=dyn_jacketBa_l_01, swing1LimitDegrees=20, swing2LimitDegrees=20, twistLimitDegrees=0}
 {bone1Name=dyn_jacketBa_l_03, bone2Name=dyn_jacketBa_l_02, swing1LimitDegrees=20, swing2LimitDegrees=20, twistLimitDegrees=0}
 {bone1Name=dyn_jacketBa_l_04, bone2Name=dyn_jacketBa_l_03, swing1LimitDegrees=20, swing2LimitDegrees=20, twistLimitDegrees=0}
 {bone1Name=dyn_jacketC_l_01, bone2Name=MSTR_jacketC_l, swing1LimitDegrees=20, swing2LimitDegrees=20, twistLimitDegrees=0}
 {bone1Name=dyn_jacketC_l_02, bone2Name=dyn_jacketC_l_01, swing1LimitDegrees=20, swing2LimitDegrees=20, twistLimitDegrees=0}
 {bone1Name=dyn_jacketC_l_03, bone2Name=dyn_jacketC_l_02, swing1LimitDegrees=20, swing2LimitDegrees=20, twistLimitDegrees=0}
 {bone1Name=dyn_jacketC_l_04, bone2Name=dyn_jacketC_l_03, swing1LimitDegrees=20, swing2LimitDegrees=20, twistLimitDegrees=0}
 {bone1Name=dyn_jacketC_r_01, bone2Name=MSTR_jacketC_r, swing1LimitDegrees=20, swing2LimitDegrees=20, twistLimitDegrees=0}
 {bone1Name=dyn_jacketC_r_02, bone2Name=dyn_jacketC_r_01, swing1LimitDegrees=20, swing2LimitDegrees=20, twistLimitDegrees=0}
 {bone1Name=dyn_jacketC_r_03, bone2Name=dyn_jacketC_r_02, swing1LimitDegrees=20, swing2LimitDegrees=20, twistLimitDegrees=0}
 {bone1Name=dyn_jacketC_r_04, bone2Name=dyn_jacketC_r_03, swing1LimitDegrees=20, swing2LimitDegrees=20, twistLimitDegrees=0}
 {bone1Name=dyn_jacketA_r_02, bone2Name=dyn_jacketA_r_01, swing1LimitDegrees=10, swing2LimitDegrees=10, twistLimitDegrees=0}
 {bone1Name=dyn_jacketA_r_03, bone2Name=dyn_jacketA_r_02, swing1LimitDegrees=10, swing2LimitDegrees=10, twistLimitDegrees=0}
 {bone1Name=dyn_jacketA_r_04, bone2Name=dyn_jacketA_r_03, swing1LimitDegrees=10, swing2LimitDegrees=10, twistLimitDegrees=0}
 {bone1Name=dyn_jacketA_l_05, bone2Name=dyn_jacketA_l_04, swing1LimitDegrees=20, swing2LimitDegrees=20, twistLimitDegrees=0}
 {bone1Name=dyn_jacketA_r_05, bone2Name=dyn_jacketA_r_04, swing1LimitDegrees=10, swing2LimitDegrees=10, twistLimitDegrees=0}
 {bone1Name=dyn_jacketB_l_05, bone2Name=dyn_jacketB_l_04, swing1LimitDegrees=20, swing2LimitDegrees=20, twistLimitDegrees=0}
 {bone1Name=dyn_jacketB_l_06, bone2Name=dyn_jacketB_l_05, swing1LimitDegrees=20, swing2LimitDegrees=20, twistLimitDegrees=0}
 {bone1Name=dyn_jacketB_r_06, bone2Name=dyn_jacketB_r_05, swing1LimitDegrees=20, swing2LimitDegrees=20, twistLimitDegrees=0}
 {bone1Name=dyn_jacketB_r_05, bone2Name=dyn_jacketB_r_04, swing1LimitDegrees=20, swing2LimitDegrees=20, twistLimitDegrees=0}
 {bone1Name=dyn_jacketBa_l_06, bone2Name=dyn_jacketBa_l_05, swing1LimitDegrees=20, swing2LimitDegrees=20, twistLimitDegrees=0}
 {bone1Name=dyn_jacketBa_l_05, bone2Name=dyn_jacketBa_l_04, swing1LimitDegrees=20, swing2LimitDegrees=20, twistLimitDegrees=0}
 {bone1Name=dyn_jacketC_l_06, bone2Name=dyn_jacketC_l_05, swing1LimitDegrees=20, swing2LimitDegrees=20, twistLimitDegrees=0}
 {bone1Name=dyn_jacketC_l_05, bone2Name=dyn_jacketC_l_04, swing1LimitDegrees=20, swing2LimitDegrees=20, twistLimitDegrees=0}
 {bone1Name=dyn_jacketC_r_06, bone2Name=dyn_jacketC_r_05, swing1LimitDegrees=20, swing2LimitDegrees=20, twistLimitDegrees=0}
 {bone1Name=dyn_jacketC_r_05, bone2Name=dyn_jacketC_r_04, swing1LimitDegrees=20, swing2LimitDegrees=20, twistLimitDegrees=0}
 {bone1Name=dyn_jacketA_l_06, bone2Name=dyn_jacketA_l_05, swing1LimitDegrees=20, swing2LimitDegrees=20, twistLimitDegrees=0}
 {bone1Name=dyn_jacketA_r_06, bone2Name=dyn_jacketA_r_05, swing1LimitDegrees=10, swing2LimitDegrees=10, twistLimitDegrees=0}
 {
  bone1Name=dyn_jacketA_l_01A
  bone2Name=MSTR_jacketA_l
  swing1LimitDegrees=20
  swing1Motion=Locked
  swing2LimitDegrees=20
  swing2Motion=Locked
  twistLimitDegrees=0
  twistMotion=Locked
 }
 {bone1Name=dyn_jacketA_r_01A, bone2Name=MSTR_jacketA_r, swing1LimitDegrees=10, swing2LimitDegrees=10, twistLimitDegrees=0}
]
