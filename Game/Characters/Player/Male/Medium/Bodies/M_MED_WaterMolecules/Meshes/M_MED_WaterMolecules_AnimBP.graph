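# Animation Blueprint for the WaterMolecules player body, adding secondary motion and pose fixes to the pose copied from its parent mesh.
# A deform Control Rig feeds Anim Dynamics on hood, fur, collar and necklace bones, then a cached pose branches into blended coat RigidBody setups: standard, skydive and slide.
# Bone offsets adjust hood and front-end jacket, while a front-end idle curve layers slot animation onto the collars.
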AnimBlueprint /Game/Characters/Player/Male/Medium/Bodies/M_MED_WaterMolecules/Meshes/M_MED_WaterMolecules_AnimBP
target_skeleton=&/M_MED_WaterMolecules_Skeleton
graphs [
 AnimGraph {nodes [
   AnimGraphNode_Slot_customIdle_0 = Slot 'DefaultSlot'()
   AnimGraphNode_UseCachedPose_customIdle_0 = Use cached pose 'startpose'()
   K2Node_VariableGet_customIdle_1 = Get Custom Idle is Playing() -> (Custom Idle is Playing:float)
   AnimGraphNode_LayeredBoneBlend_customIdle_0 = Layered blend per bone(
    BasePose=@AnimGraphNode_UseCachedPose_customIdle_0
    BlendPoses_0=@AnimGraphNode_Slot_customIdle_0
    BlendWeights_0=@K2Node_VariableGet_customIdle_1
   ) {layerSetup [{branchFilters [
       {boneName=CR_collarMid_r_01}
       {boneName=CR_collarMid_l_01}
       {boneName=CR_collar_r}
       {boneName=MSTR_collar_r}
       {boneName=collar_r_01}
       {boneName=collar_r_02}
       {boneName=CR_collar_l}
       {boneName=MSTR_collar_l}
       {boneName=collar_l_01}
       {boneName=collar_l_02}
      ]}]}
   Output Pose(Result=@AnimGraphNode_LayeredBoneBlend_customIdle_0) {layerGroup=None}
   n1 = Control Rig(Source=Copy Pose From Mesh() {bUseAttachedParent=true}) {
    alphaBoolBlend {blendOption=Linear}
    controlRigAssetReference {blueprintRigClass=&/M_MED_DeformRig_WaterMolecules_CR.M_MED_DeformRig_WaterMolecules_CR_C}
   }
   n2 = "Transform (Modify) Bone - Bone: attach_cape"(ComponentPose=Local To Component(LocalPose=@n1), Alpha=1, Translation="0,-3,0", Rotation="0,0,0", Scale="1,1,1") {
    translationMode=BMM_Additive
    translationSpace=BCS_BoneSpace
    alphaBoolBlend {blendOption=Linear}
    boneToModify {boneName=attach_cape}
    translation {y=-3}
   }
   n3 = Anim Dynamics - Bone: dyn_hoodA_ba_01(ComponentPose=@n2, Alpha=1, ExternalForce="0,0,0") {
    angularDampingOverride=1
    angularSpringConstant=20000
    bAngularSpring=true
    bOverrideAngularDamping=true
    bUseGravityOverride=true
    gravityScale=0.35
    simulationSpace=RootRelative
    alphaBoolBlend {blendOption=Linear}
    boundBone {boneName=dyn_hoodA_ba_01}
    componentLinearAccScale=<1 1 1>
    componentLinearVelScale=<1 1 1>
    gravityOverride {z=-980}
    physicsBodyDefinitions [{
      boundBone {boneName=dyn_hoodA_ba_01}
      boxExtents=<50 50 50>
      constraintSetup {angularLimitsMax=<0 0 0>, angularLimitsMin=<0 -3 0>, angularTarget=<1 0 0>}
      localJointOffset=<2 0 0>
     }]
   }
   n4 = Anim Dynamics - Bone: AD_fur_0A1(ComponentPose=@n3, Alpha=1, ExternalForce="0,0,0") {
    angularDampingOverride=1
    angularSpringConstant=20000
    bAngularSpring=true
    bOverrideAngularDamping=true
    bUseGravityOverride=true
    gravityScale=0.35
    simulationSpace=RootRelative
    alphaBoolBlend {blendOption=Linear}
    boundBone {boneName=AD_fur_0A1}
    componentLinearAccScale=<1 1 1>
    componentLinearVelScale=<1 1 1>
    gravityOverride {z=-980}
    physicsBodyDefinitions [{
      boundBone {boneName=AD_fur_0A1}
      boxExtents=<30 40 40>
      constraintSetup {angularLimitsMax=<0 4 4>, angularLimitsMin=<0 -4 -4>, angularTarget=<1 0 0>}
      localJointOffset=<3 0 0>
     }]
   }
   n5 = Anim Dynamics - Bone: AD_fur_0B1(ComponentPose=@n4, Alpha=1, ExternalForce="0,0,0") {
    angularDampingOverride=1
    angularSpringConstant=20000
    bAngularSpring=true
    bOverrideAngularDamping=true
    bUseGravityOverride=true
    gravityScale=0.35
    simulationSpace=RootRelative
    alphaBoolBlend {blendOption=Linear}
    boundBone {boneName=AD_fur_0B1}
    componentLinearAccScale=<1 1 1>
    componentLinearVelScale=<1 1 1>
    gravityOverride {z=-980}
    physicsBodyDefinitions [{
      boundBone {boneName=AD_fur_0B1}
      boxExtents=<40 50 50>
      constraintSetup {angularLimitsMax=<0 3 3>, angularLimitsMin=<0 -3 -3>, angularTarget=<1 0 0>}
      localJointOffset=<3 0 0>
     }]
   }
   n6 = Anim Dynamics - Bone: AD_fur_0C1(ComponentPose=@n5, Alpha=1, ExternalForce="0,0,0") {
    angularDampingOverride=1
    angularSpringConstant=20000
    bAngularSpring=true
    bOverrideAngularDamping=true
    bUseGravityOverride=true
    gravityScale=0.35
    simulationSpace=RootRelative
    alphaBoolBlend {blendOption=Linear}
    boundBone {boneName=AD_fur_0C1}
    componentLinearAccScale=<1 1 1>
    componentLinearVelScale=<1 1 1>
    gravityOverride {z=-980}
    physicsBodyDefinitions [{
      boundBone {boneName=AD_fur_0C1}
      boxExtents=<25 35 35>
      constraintSetup {angularLimitsMax=<0 2 2>, angularLimitsMin=<0 -2 -2>, angularTarget=<1 0 0>}
      localJointOffset=<3 0 0>
     }]
   }
   n7 = Anim Dynamics - Bone: AD_fur_0D1(ComponentPose=@n6, Alpha=1, ExternalForce="0,0,0") {
    angularDampingOverride=1
    angularSpringConstant=20000
    bAngularSpring=true
    bOverrideAngularDamping=true
    bUseGravityOverride=true
    gravityScale=0.35
    simulationSpace=RootRelative
    alphaBoolBlend {blendOption=Linear}
    boundBone {boneName=AD_fur_0D1}
    componentLinearAccScale=<1 1 1>
    componentLinearVelScale=<1 1 1>
    gravityOverride {z=-980}
    physicsBodyDefinitions [{
      boundBone {boneName=AD_fur_0D1}
      boxExtents=<25 35 35>
      constraintSetup {angularLimitsMax=<0 4 4>, angularLimitsMin=<0 -4 -4>, angularTarget=<1 0 0>}
      localJointOffset=<3 0 0>
     }]
   }
   n8 = Anim Dynamics - Bone: AD_fur_0E1(ComponentPose=@n7, Alpha=1, ExternalForce="0,0,0") {
    angularDampingOverride=1
    angularSpringConstant=20000
    bAngularSpring=true
    bOverrideAngularDamping=true
    bUseGravityOverride=true
    gravityScale=0.35
    simulationSpace=RootRelative
    alphaBoolBlend {blendOption=Linear}
    boundBone {boneName=AD_fur_0E1}
    componentLinearAccScale=<1 1 1>
    componentLinearVelScale=<1 1 1>
    gravityOverride {z=-980}
    physicsBodyDefinitions [{
      boundBone {boneName=AD_fur_0E1}
      boxExtents=<40 50 50>
      constraintSetup {angularLimitsMax=<0 3 3>, angularLimitsMin=<0 -3 -3>, angularTarget=<1 0 0>}
      localJointOffset=<3 0 0>
     }]
   }
   n9 = Anim Dynamics - Bone: AD_fur_0F1(ComponentPose=@n8, Alpha=1, ExternalForce="0,0,0") {
    angularDampingOverride=1
    angularSpringConstant=20000
    bAngularSpring=true
    bOverrideAngularDamping=true
    bUseGravityOverride=true
    gravityScale=0.35
    simulationSpace=RootRelative
    alphaBoolBlend {blendOption=Linear}
    boundBone {boneName=AD_fur_0F1}
    componentLinearAccScale=<1 1 1>
    componentLinearVelScale=<1 1 1>
    gravityOverride {z=-980}
    physicsBodyDefinitions [{
      boundBone {boneName=AD_fur_0F1}
      boxExtents=<30 40 40>
      constraintSetup {angularLimitsMax=<0 2 2>, angularLimitsMin=<0 -2 -2>, angularTarget=<1 0 0>}
      localJointOffset=<3 0 0>
     }]
   }
   n10 = Anim Dynamics - Bone: AD_fur_0G1(ComponentPose=@n9, Alpha=1, ExternalForce="0,0,0") {
    angularDampingOverride=1
    angularSpringConstant=20000
    bAngularSpring=true
    bOverrideAngularDamping=true
    bUseGravityOverride=true
    gravityScale=0.35
    simulationSpace=RootRelative
    alphaBoolBlend {blendOption=Linear}
    boundBone {boneName=AD_fur_0G1}
    componentLinearAccScale=<1 1 1>
    componentLinearVelScale=<1 1 1>
    gravityOverride {z=-980}
    physicsBodyDefinitions [{
      boundBone {boneName=AD_fur_0G1}
      boxExtents=<15 25 25>
      constraintSetup {angularLimitsMax=<0 4 4>, angularLimitsMin=<0 -4 -4>, angularTarget=<1 0 0>}
      localJointOffset=<3 0 0>
     }]
   }
   n11 = Anim Dynamics - Bone: AD_fur_0H1(ComponentPose=@n10, Alpha=1, ExternalForce="0,0,0") {
    angularDampingOverride=1
    angularSpringConstant=20000
    bAngularSpring=true
    bOverrideAngularDamping=true
    bUseGravityOverride=true
    gravityScale=0.35
    simulationSpace=RootRelative
    alphaBoolBlend {blendOption=Linear}
    boundBone {boneName=AD_fur_0H1}
    componentLinearAccScale=<1 1 1>
    componentLinearVelScale=<1 1 1>
    gravityOverride {z=-980}
    physicsBodyDefinitions [{
      boundBone {boneName=AD_fur_0H1}
      boxExtents=<40 50 50>
      constraintSetup {angularLimitsMax=<0 3 3>, angularLimitsMin=<0 -3 -3>, angularTarget=<1 0 0>}
      localJointOffset=<3 0 0>
     }]
   }
   n12 = Anim Dynamics - Bone: AD_fur_0I1(ComponentPose=@n11, Alpha=1, ExternalForce="0,0,0") {
    angularDampingOverride=1
    angularSpringConstant=20000
    bAngularSpring=true
    bOverrideAngularDamping=true
    bUseGravityOverride=true
    gravityScale=0.35
    simulationSpace=RootRelative
    alphaBoolBlend {blendOption=Linear}
    boundBone {boneName=AD_fur_0I1}
    componentLinearAccScale=<1 1 1>
    componentLinearVelScale=<1 1 1>
    gravityOverride {z=-980}
    physicsBodyDefinitions [{
      boundBone {boneName=AD_fur_0I1}
      boxExtents=<40 50 50>
      constraintSetup {angularLimitsMax=<0 2 2>, angularLimitsMin=<0 -2 -2>, angularTarget=<1 0 0>}
      localJointOffset=<3 0 0>
     }]
   }
   n13 = Anim Dynamics - Bone: AD_fur_0J1(ComponentPose=@n12, Alpha=1, ExternalForce="0,0,0") {
    angularDampingOverride=1
    angularSpringConstant=20000
    bAngularSpring=true
    bOverrideAngularDamping=true
    bUseGravityOverride=true
    gravityScale=0.35
    simulationSpace=RootRelative
    alphaBoolBlend {blendOption=Linear}
    boundBone {boneName=AD_fur_0J1}
    componentLinearAccScale=<1 1 1>
    componentLinearVelScale=<1 1 1>
    gravityOverride {z=-980}
    physicsBodyDefinitions [{
      boundBone {boneName=AD_fur_0J1}
      boxExtents=<30 40 40>
      constraintSetup {angularLimitsMax=<0 4 4>, angularLimitsMin=<0 -4 -4>, angularTarget=<1 0 0>}
      localJointOffset=<3 0 0>
     }]
   }
   n14 = Anim Dynamics - Bone: AD_fur_0K1(ComponentPose=@n13, Alpha=1, ExternalForce="0,0,0") {
    angularDampingOverride=1
    angularSpringConstant=20000
    bAngularSpring=true
    bOverrideAngularDamping=true
    bUseGravityOverride=true
    gravityScale=0.35
    simulationSpace=RootRelative
    alphaBoolBlend {blendOption=Linear}
    boundBone {boneName=AD_fur_0K1}
    componentLinearAccScale=<1 1 1>
    componentLinearVelScale=<1 1 1>
    gravityOverride {z=-980}
    physicsBodyDefinitions [{
      boundBone {boneName=AD_fur_0K1}
      boxExtents=<40 50 50>
      constraintSetup {angularLimitsMax=<0 3 3>, angularLimitsMin=<0 -3 -3>, angularTarget=<1 0 0>}
      localJointOffset=<3 0 0>
     }]
   }
   n15 = Anim Dynamics - Bone: AD_fur_0L1(ComponentPose=@n14, Alpha=1, ExternalForce="0,0,0") {
    angularDampingOverride=1
    angularSpringConstant=20000
    bAngularSpring=true
    bOverrideAngularDamping=true
    bUseGravityOverride=true
    gravityScale=0.35
    simulationSpace=RootRelative
    alphaBoolBlend {blendOption=Linear}
    boundBone {boneName=AD_fur_0L1}
    componentLinearAccScale=<1 1 1>
    componentLinearVelScale=<1 1 1>
    gravityOverride {z=-980}
    physicsBodyDefinitions [{
      boundBone {boneName=AD_fur_0L1}
      boxExtents=<25 35 35>
      constraintSetup {angularLimitsMax=<0 2 2>, angularLimitsMin=<0 -2 -2>, angularTarget=<1 0 0>}
      localJointOffset=<3 0 0>
     }]
   }
   n16 = Anim Dynamics - Bone: AD_fur_0M1(ComponentPose=@n15, Alpha=1, ExternalForce="0,0,0") {
    angularDampingOverride=1
    angularSpringConstant=20000
    bAngularSpring=true
    bOverrideAngularDamping=true
    bUseGravityOverride=true
    gravityScale=0.35
    simulationSpace=RootRelative
    alphaBoolBlend {blendOption=Linear}
    boundBone {boneName=AD_fur_0M1}
    componentLinearAccScale=<1 1 1>
    componentLinearVelScale=<1 1 1>
    gravityOverride {z=-980}
    physicsBodyDefinitions [{
      boundBone {boneName=AD_fur_0M1}
      boxExtents=<40 50 50>
      constraintSetup {angularLimitsMax=<0 4 4>, angularLimitsMin=<0 -4 -4>, angularTarget=<1 0 0>}
      localJointOffset=<3 0 0>
     }]
   }
   n17 = Anim Dynamics - Bone: AD_fur_0N1(ComponentPose=@n16, Alpha=1, ExternalForce="0,0,0") {
    angularDampingOverride=1
    angularSpringConstant=20000
    bAngularSpring=true
    bOverrideAngularDamping=true
    bUseGravityOverride=true
    gravityScale=0.35
    simulationSpace=RootRelative
    alphaBoolBlend {blendOption=Linear}
    boundBone {boneName=AD_fur_0N1}
    componentLinearAccScale=<1 1 1>
    componentLinearVelScale=<1 1 1>
    gravityOverride {z=-980}
    physicsBodyDefinitions [{
      boundBone {boneName=AD_fur_0N1}
      boxExtents=<25 35 35>
      constraintSetup {angularLimitsMax=<0 3 3>, angularLimitsMin=<0 -3 -3>, angularTarget=<1 0 0>}
      localJointOffset=<3 0 0>
     }]
   }
   n18 = Anim Dynamics - Bone: AD_fur_0O1(ComponentPose=@n17, Alpha=1, ExternalForce="0,0,0") {
    angularDampingOverride=1
    angularSpringConstant=20000
    bAngularSpring=true
    bOverrideAngularDamping=true
    bUseGravityOverride=true
    gravityScale=0.35
    simulationSpace=RootRelative
    alphaBoolBlend {blendOption=Linear}
    boundBone {boneName=AD_fur_0O1}
    componentLinearAccScale=<1 1 1>
    componentLinearVelScale=<1 1 1>
    gravityOverride {z=-980}
    physicsBodyDefinitions [{
      boundBone {boneName=AD_fur_0O1}
      boxExtents=<28 38 38>
      constraintSetup {angularLimitsMax=<0 2 2>, angularLimitsMin=<0 -2 -2>, angularTarget=<1 0 0>}
      localJointOffset=<3 0 0>
     }]
   }
   n19 = Anim Dynamics - Bone: AD_fur_0P1(ComponentPose=@n18, Alpha=1, ExternalForce="0,0,0") {
    angularDampingOverride=1
    angularSpringConstant=20000
    bAngularSpring=true
    bOverrideAngularDamping=true
    bUseGravityOverride=true
    gravityScale=0.35
    simulationSpace=RootRelative
    alphaBoolBlend {blendOption=Linear}
    boundBone {boneName=AD_fur_0P1}
    componentLinearAccScale=<1 1 1>
    componentLinearVelScale=<1 1 1>
    gravityOverride {z=-980}
    physicsBodyDefinitions [{
      boundBone {boneName=AD_fur_0P1}
      boxExtents=<26 36 36>
      constraintSetup {angularLimitsMax=<0 4 4>, angularLimitsMin=<0 -4 -4>, angularTarget=<1 0 0>}
      localJointOffset=<3 0 0>
     }]
   }
   n20 = Anim Dynamics - Bone: AD_fur_0Q1(ComponentPose=@n19, Alpha=1, ExternalForce="0,0,0") {
    angularDampingOverride=1
    angularSpringConstant=20000
    bAngularSpring=true
    bOverrideAngularDamping=true
    bUseGravityOverride=true
    gravityScale=0.35
    simulationSpace=RootRelative
    alphaBoolBlend {blendOption=Linear}
    boundBone {boneName=AD_fur_0Q1}
    componentLinearAccScale=<1 1 1>
    componentLinearVelScale=<1 1 1>
    gravityOverride {z=-980}
    physicsBodyDefinitions [{
      boundBone {boneName=AD_fur_0Q1}
      boxExtents=<40 50 50>
      constraintSetup {angularLimitsMax=<0 3 3>, angularLimitsMin=<0 -3 -3>, angularTarget=<1 0 0>}
      localJointOffset=<3 0 0>
     }]
   }
   n21 = Anim Dynamics - Bone: AD_furR_01(ComponentPose=@n20, Alpha=1, ExternalForce="0,0,0") {
    angularDampingOverride=1
    angularSpringConstant=20000
    bAngularSpring=true
    bOverrideAngularDamping=true
    bUseGravityOverride=true
    gravityScale=0.35
    simulationSpace=RootRelative
    alphaBoolBlend {blendOption=Linear}
    boundBone {boneName=AD_furR_01}
    componentLinearAccScale=<1 1 1>
    componentLinearVelScale=<1 1 1>
    gravityOverride {z=-980}
    physicsBodyDefinitions [{
      boundBone {boneName=AD_furR_01}
      boxExtents=<30 40 40>
      constraintSetup {angularLimitsMax=<0 2 2>, angularLimitsMin=<0 -2 -2>, angularTarget=<1 0 0>}
      localJointOffset=<3 0 0>
     }]
   }
   n22 = Anim Dynamics - Bone: collar_l_01(ComponentPose=@n21, Alpha=1, ExternalForce="0,0,0") {
    angularDampingOverride=1
    angularSpringConstant=20000
    bAngularSpring=true
    bOverrideAngularDamping=true
    bUseGravityOverride=true
    gravityScale=0.35
    simulationSpace=RootRelative
    alphaBoolBlend {blendOption=Linear}
    boundBone {boneName=collar_l_01}
    componentLinearAccScale=<1 1 1>
    componentLinearVelScale=<1 1 1>
    gravityOverride {z=-980}
    physicsBodyDefinitions [{
      boundBone {boneName=collar_l_01}
      boxExtents=<40 50 50>
      constraintSetup {angularLimitsMax=<15 15 4>, angularLimitsMin=<-15 -15 -4>, angularTarget=<1 0 0>}
      localJointOffset=<3 0 0>
     }]
   }
   n23 = Anim Dynamics - Bone: collar_r_01(ComponentPose=@n22, Alpha=1, ExternalForce="0,0,0") {
    angularDampingOverride=1
    angularSpringConstant=20000
    bAngularSpring=true
    bOverrideAngularDamping=true
    bUseGravityOverride=true
    gravityScale=0.35
    simulationSpace=RootRelative
    alphaBoolBlend {blendOption=Linear}
    boundBone {boneName=collar_r_01}
    componentLinearAccScale=<1 1 1>
    componentLinearVelScale=<1 1 1>
    gravityOverride {z=-980}
    physicsBodyDefinitions [{
      boundBone {boneName=collar_r_01}
      boxExtents=<40 50 50>
      constraintSetup {angularLimitsMax=<0 0 4>, angularLimitsMin=<0 0 -4>, angularTarget=<1 0 0>}
      localJointOffset=<3 0 0>
     }]
   }
   n24 = Anim Dynamics - Bone: MSTR_necklaceCord(ComponentPose=@n23, Alpha=1, ExternalForce="0,0,0") {
    angularDampingOverride=1
    angularSpringConstant=20000
    bAngularSpring=true
    bOverrideAngularDamping=true
    bUseGravityOverride=true
    gravityScale=0.35
    simulationSpace=RootRelative
    alphaBoolBlend {blendOption=Linear}
    boundBone {boneName=MSTR_necklaceCord}
    componentLinearAccScale=<1 1 1>
    componentLinearVelScale=<1 1 1>
    gravityOverride {z=-980}
    physicsBodyDefinitions [{
      boundBone {boneName=MSTR_necklaceCord}
      boxExtents=<20 30 30>
      constraintSetup {angularLimitsMax=<0 0 15>, angularLimitsMin=<0 0 -15>, angularTarget=<1 0 0>}
      localJointOffset=<4 0 0>
     }]
   }
   n25 = Anim Dynamics - Bone: MSTR_necklace(ComponentPose=@n24, Alpha=1, ExternalForce="0,0,0") {
    angularDampingOverride=1
    angularSpringConstant=20000
    bAngularSpring=true
    bOverrideAngularDamping=true
    bUseGravityOverride=true
    gravityScale=0.35
    simulationSpace=RootRelative
    alphaBoolBlend {blendOption=Linear}
    boundBone {boneName=MSTR_necklace}
    componentLinearAccScale=<1 1 1>
    componentLinearVelScale=<1 1 1>
    gravityOverride {z=-980}
    physicsBodyDefinitions [{
      boundBone {boneName=MSTR_necklace}
      boxExtents=<20 30 30>
      constraintSetup {angularLimitsMax=<0 0 0>, angularLimitsMin=<0 -10 0>, angularTarget=<1 0 0>}
      localJointOffset=<4 0 0>
     }]
   }
   Save cached pose 'Skirt Cache'(Pose=Component To Local(ComponentPose=@n25)) {cacheName=Skirt Cache}
   n26 = Get GravityOverride()
   n27 = RigidBody!(
    ComponentPose=Local To Component(LocalPose=Use cached pose 'Skirt Cache'())
    Alpha
    OverrideWorldGravity:Vector=@n26
    ExternalForce="0,0,-980"
   ) -> (Pose) {
    bClampLinearTranslationLimitToRefPose=true
    bOverrideWorldGravity=true
    overridePhysicsAsset=&/M_MED_WaterMolecules_Coat_Physics
    simulationSpace=BaseBoneSpace
    alphaBoolBlend {blendOption=Linear}
    baseBoneRef {boneName=pelvis}
    componentAppliedLinearAccClamp=<3000 3000 3000>
    componentLinearAccScale=<0.4 0.3 0.3>
    componentLinearVelScale=<0.4 0.3 0.3>
    overrideWorldGravity {z=-980}
   }
   n28 = RigidBody(ComponentPose=Local To Component(LocalPose=Use cached pose 'Skirt Cache'()), OverrideWorldGravity=@n26, ExternalForce="0,0,-980") {
    bClampLinearTranslationLimitToRefPose=true
    bOverrideWorldGravity=true
    overridePhysicsAsset=&/M_MED_WaterMolecules_Coat_SD_Physics
    simulationSpace=BaseBoneSpace
    alphaBoolBlend {blendOption=Linear}
    baseBoneRef {boneName=pelvis}
    componentAppliedLinearAccClamp=<3000 3000 3000>
    componentLinearAccScale=<0.4 0.3 0.3>
    componentLinearVelScale=<0.4 0.3 0.3>
    overrideWorldGravity {z=-980}
   }
   n29 = Blend Poses by bool(BlendPose_0=Component To Local(ComponentPose=@n28), BlendPose_1=Component To Local(ComponentPose=@n27), bActiveValue=Get bIsSkyDiving()) {blendType=Linear}
   n30 = RigidBody(
    ComponentPose=Local To Component(LocalPose=Use cached pose 'Skirt Cache'())
    OverrideWorldGravity=Get GravityOverride()
    ExternalForce="0,0,-980"
   ) {
    bClampLinearTranslationLimitToRefPose=true
    bOverrideWorldGravity=true
    overridePhysicsAsset=&/M_MED_WaterMolecules_Coat_Slide_Physics
    simulationSpace=BaseBoneSpace
    alphaBoolBlend {blendOption=Linear}
    baseBoneRef {boneName=pelvis}
    componentAppliedLinearAccClamp=<3000 3000 3000>
    componentLinearAccScale=<0.4 0.3 0.3>
    componentLinearVelScale=<0.4 0.3 0.3>
    overrideWorldGravity {z=-980}
   }
   n31 = Component To Local(ComponentPose=@n30)
   n32 = Get isFrontEndNotEmote() -> (isFrontEndNotEmote:bool)
   n33 = "Transform (Modify) Bone - Bone: CR_rot_jacketA_l"!(
    ComponentPose=Local To Component(LocalPose=Blend Poses by bool(BlendPose_0=@n31, BlendPose_1=@n29, BlendTime_1=1) {blendType=Linear})
    bAlphaBoolEnabled:bool=@n32
    Translation="0,0,0"
    Rotation="0,0,0"
    Scale="1,1,1"
   ) -> (Pose) {
    alphaInputType=Bool
    rotationMode=BMM_Additive
    rotationSpace=BCS_BoneSpace
    scaleSpace=BCS_BoneSpace
    translationMode=BMM_Additive
    translationSpace=BCS_BoneSpace
    alphaBoolBlend {blendOption=Linear}
    boneToModify {boneName=CR_rot_jacketA_l}
   }
   n34 = "Transform (Modify) Bone - Bone: CR_rot_jacketA_r"!(ComponentPose=@n33, bAlphaBoolEnabled:bool=@n32, Translation="0,0,2", Rotation="5,-5,0", Scale="1,1,1") -> (Pose) {
    alphaInputType=Bool
    rotationMode=BMM_Additive
    rotationSpace=BCS_BoneSpace
    translationMode=BMM_Additive
    translationSpace=BCS_BoneSpace
    alphaBoolBlend {blendOption=Linear}
    boneToModify {boneName=CR_rot_jacketA_r}
    rotation {pitch=5, yaw=-5}
    translation {z=2}
   }
   n35 = "Transform (Modify) Bone - Bone: CR_neckHood_all"!(ComponentPose=@n34, bAlphaBoolEnabled:bool=True, Translation="0,0,0", Rotation="0,-18,0", Scale="1,1,1") -> (Pose) {
    alphaInputType=Bool
    rotationMode=BMM_Additive
    rotationSpace=BCS_BoneSpace
    translationMode=BMM_Additive
    translationSpace=BCS_BoneSpace
    alphaBoolBlend {blendOption=Linear}
    boneToModify {boneName=CR_neckHood_all}
    rotation {yaw=-18}
   }
   AnimGraphNode_SaveCachedPose_customIdle_0 = Save cached pose 'startpose'(Pose=Component To Local(ComponentPose=@n35)) {cacheName=startpose}
  ]}
 EventGraph {nodes [
   n1 = Cast To FortPlayerPawn(Object=GetOwningActor())
   > Set currentBackpack(currentBackpack:Athena Backpack Item Definition Object Reference=GetItemDefinitionTypeFromLoadoutWithActor!(
     Actor:Actor Object Reference=@n1
     LoadoutSlot:Gameplay Tag Structure=(TagName="LoadoutSlot.Backpack")
     CosmeticType:Athena Cosmetic Item Definition Class Reference=/Script/FortniteGame.AthenaBackpackItemDefinition
    ) -> (OutCosmeticItemDef:Athena Backpack Item Definition Object Reference)) -> (Output_Get:Athena Backpack Item Definition Object Reference)
   Event BlueprintUpdateAnimation()
   > n2 = Sequence!(execute) -> (then_0, then_1, then_2:Exec, then_3:Exec)
   > n3 = Gravity Override State Function DEPRECATED(CharacterPartAnimInstance=Self-Reference(), GameTimeInSeconds=GetGameTimeInSeconds(), GravityOverrideParams=Get Gravity Override Params())
   > Set GravityOverride(GravityOverride=@n3)
   K2Node_VariableGet_customIdle_0 = Get bIsInFrontEnd()
   K2Node_IfThenElse_customIdle_0 = Branch(execute=@n2.then_1, Condition=@K2Node_VariableGet_customIdle_0)
   K2Node_CallFunction_customIdle_0 = GetOwnerFrontEndAnimInstance()
   K2Node_CallFunction_customIdle_1 = GetCurveValue(self=@K2Node_CallFunction_customIdle_0, CurveName=Frontend_Idle_Bypass)
   > K2Node_VariableSet_customIdle_0 = Set Custom Idle is Playing(Custom Idle is Playing:float=@K2Node_CallFunction_customIdle_1) -> (Output_Get:float)
   n4 = Get bIsJumping()
   n5 = Get bIsTacticalSprinting()
   Set IsSlidingOrSprintJumping(execute=@n2.then_2, IsSlidingOrSprintJumping:bool=OR Boolean!(A=Get bIsSliding(), B=AND Boolean(A=@n5, B=@n4), C:bool) -> (ReturnValue)) -> (Output_Get:bool)
   n6 = Get bIsPlayingEmote()
   Set isFrontEndNotEmote(execute=@n2.then_3, isFrontEndNotEmote:bool=AND Boolean(A=Get bIsInFrontEnd(), B=NOT Boolean(A=@n6))) -> (Output_Get:bool)
  ]}
]
variables [
 Backbling Item Def {type=AthenaBackpackItemDefinition, default=null}
 Current Backpack {type=AthenaBackpackItemDefinition, default=null}
 currentBackpack {type=AthenaBackpackItemDefinition, default=null}
 currentGlider {type=AthenaBackpackItemDefinition, default=null}
 Custom Idle is Playing {type=number}
 Gravity Override Params {type=GravityOverrideParamsStruct}
 GravityOverride {type=Vector}
 HasPack {type=boolean}
 isFrontEndNotEmote {type=boolean}
 IsSlidingOrSprintJumping {type=boolean}
]
